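# Translucent unlit material for a lighting card on the bowling alley ball return prop, apparently faking a soft colored glow, reddish by default.
# A saturated depth fade softens the card near intersecting geometry, driving opacity through a power curve and scalar multiplier.
# The same fade scales the squared Color parameter from unchanged up to ColorMultDeep, with the result clamped by Saturate, so the HDR default color loses its over-bright range.
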Material /Game/Environments/Asteria/Props/BowlingAlley/BowlingAlley_BallReturn_A/Materials/M_BowleyAlley_BallReturn_LightingCard
graph {
 properties {bCastRayTracedShadows=false, blendMode=BLEND_Translucent, pixelDepthOffsetMode=PDOM_Legacy, shadingModel=MSM_Unlit, usedShadingModels=Unlit}
 nodes [
  n1 = VectorParameter() {parameterName=Color, defaultValue=<rgba 5 0.1551 0.2546 1>}
  n2 = ScalarParameter() {defaultValue=2, parameterName=ColorMultDeep}
  n3 = Saturate(None=DepthFade(FadeDistance=ScalarParameter() {defaultValue=150, parameterName=DepthFadeDist}))
  Saturate(None=Multiply(A=Multiply(A=@n1.RGB, B=@n1.RGB), B=LinearInterpolate(B=@n2, Alpha=@n3) {constA=1}))
  n4 = ScalarParameter() {defaultValue=1.5, parameterName=DepthPower}
  Saturate(None=Multiply(A=ScalarParameter() {defaultValue=0.15, parameterName=Opacity}, B=Power(Base=@n3, Exp=@n4)))
 ]
}
